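# Niagara system that draws a retro arcade space-shooter game (ships, bullets, explosion, score and text) for the Interstellar Sprites animation prop.
# Emitters pin mesh and sprite particles to bones of User.SkelMesh, remapped by system game scale and offset.
# Sampled bone scale acts as an animation-driven signal for visibility and sub-image frames, while bullets also hide by height.
# User.PointsCurve triggers Score increments, displayed as per-digit particles.
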NiagaraSystem /BRCosmetics/Animation/Props/Gadgets_Interstellar_Sprites/FX/NS_InterstellarSprites
dependencies {
 usedDataInterfaces [&NiagaraDataInterfaceCurve, &NiagaraDataInterfaceArrayInt32]
 usedDynamicInputs [&NormalizeVector]
 usedModules [
  &self:SystemSpawnScript.SystemScriptSource.SystemScriptGraph.NiagaraNodeAssignment_0.SetVariables_#1
  &self:SystemSpawnScript.SystemScriptSource.SystemScriptGraph.NiagaraNodeAssignment_1.SetVariables_#2
  &self:SystemSpawnScript.SystemScriptSource.SystemScriptGraph.NiagaraNodeAssignment_2.SetVariables_#3
  &SystemState
 ]
}
emitters {
 1UPText {
  emitterUpdateScript [
   &EmitterState()
   &SpawnBurst_Instantaneous(Use Spawn Probability=false, UseLoopCountLimit=false)
  ]
  particleSpawnScript [
   &InitializeParticle(
    Lifetime=5
    Mesh Scale Mode=::NewEnumerator3
    Mesh Uniform Scale=&Multiply_Float(A=1, B=$System.ObjectScale)
    UsePositionOffset=false
    Write Material Random=true
   )
   &SkeletalMeshLocation(
    Bone / Socket Index=27
    Bone Sampling Mode=::NewEnumerator4
    MaskSpawnGroup=false
    Position Sampling=::NewEnumerator2
    Sampled Position Offset=<1.75 0 0>
    Skeletal Mesh=$User.SkelMesh
   ) {module=SkeletalMeshLocation001}
   /SetVariables(Particles.Position=&Add_Vector(
     A=&Add_Vector(
      A=&Multiply_VectorByFloat(
       Vector=Linked() {linkedVariable {name=Particles.SkeletalMeshLocation001.SampledPosition, type=NiagaraPosition}}
       Float=$System.GlobalGameScale
      )
      B=&RotateVectorByQuaternion(
       Vector To Rotate=&Multiply_Vector(A=<1 0 -20>, B=&VectorFromFloat(Value=$System.GlobalGameScale))
       Quaternion=$Particles.SkeletalMeshLocation001.SampledRotation
      )
     )
     B=$System.GlobalGameOffset
    )) {isSetParameters=true, module=SetVariables_#4}
   /SetVariables(Particles.MeshOrientation=&MultiplyQuaternion(
     Quaternion A=$Particles.SkeletalMeshLocation001.SampledRotation
     Quaternion B=&MultiplyQuaternion(
      Quaternion A=&MakeQuatFromAxisAngle(Axis=<1 0 0>, AngleInDegrees=-90)
      Quaternion B=&MakeQuatFromAxisAngle(Axis=<0 1 0>, AngleInDegrees=90)
     )
    )) {isSetParameters=true, module=SetVariables_#5}
  ]
  particleUpdateScript [
   &ParticleState()
   &ScaleColor(
    Scale Alpha=/Niagara/DynamicInputs/Integer/V2/Round_Float(Float:NiagaraFloat=&Frac_Float(Float=&Multiply_Float(A=$Emitter.Age, B=2)))
    ScaleA=true
    ScaleRGB=true
   )
  ]
  rendererClasses [&NiagaraMeshRendererProperties]
  renderers [{rendererClass=&NiagaraMeshRendererProperties, rendererIndex=0}]
 }
 Bullets {
  emitterUpdateScript [
   &EmitterState()
   &SpawnBurst_Instantaneous(Spawn Count=4, Use Spawn Probability=false, UseLoopCountLimit=false)
  ]
  particleSpawnScript [
   &InitializeParticle(
    Lifetime=5
    Mesh Scale=&Multiply_VectorByFloat(Vector=<3 1 1>, Float=$System.ObjectScale)
    Mesh Scale Mode=::NewEnumerator4
    Sprite Size=&Multiply_Vector2D_ByFloat(Float=$System.ObjectScale, Vector2D=<24 22>)
    Sprite Size Mode=::NewEnumerator4
    UsePositionOffset=false
    Write Material Random=true
   )
   /SetVariables(Particles.SocketIndex:NiagaraInt32=&SelectIntFromArray(
     Direct Array Index=&ReturnExecIndex()
     Int Selection Array=DataInterface() {propertyValues {IntData [20,22,23,24]}}
    )) {isSetParameters=true, module=SetVariables_#6}
   /SetVariables(Particles.Color=&MakeCustomLinearColorFromBool(
     Bool=&SetBoolByIntComparison(A=$Particles.SocketIndex, B=24, Comparison Type=::NewEnumerator4)
     True Color=<rgba 0.8714 0.8308 0.0595 1>
     False Color=<rgba 1 1 1 1>
    )) {isSetParameters=true}
   /Game/Animation/Props/Gadgets_Interstellar_Sprites/FX/Modules/SkeletalMeshLocation_WithScale(
    Bone / Socket Index:NiagaraInt32=$Particles.SocketIndex
    Bone Sampling Mode:ENiagaraBoneSamplingMode=::NewEnumerator4
    MaskSpawnGroup=false
    Position Sampling:ENiagaraSkelMeshPositionSamplingMode=::NewEnumerator2
    Skeletal Mesh:NiagaraDataInterfaceSkeletalMesh=$User.SkelMesh
   ) {module=SkeletalMeshLocation_WithScale001}
   /SetVariables(Particles.Position=&Add_Vector(
     A=&Add_Vector(
      A=&Multiply_VectorByFloat(Vector=$Particles.SkeletalMeshLocation_WithScale001.SampledPosition, Float=$System.GlobalGameScale)
      B=&RotateVectorByQuaternion(
       Vector To Rotate=&Multiply_Vector(A=<-0.5 0 0>, B=&VectorFromFloat(Value=$System.GlobalGameScale))
       Quaternion=$Particles.SkeletalMeshLocation_WithScale001.SampledRotation
      )
     )
     B=$System.GlobalGameOffset
    )) {isSetParameters=true}
   /SetVariables(Particles.SubImageIndex=5) {isSetParameters=true}
   /SetVariables(Particles.SpriteAlignment=<0 0.3713 0.9284>, Particles.SpriteFacing=$System.SpriteFacing) {isSetParameters=true}
  ]
  particleUpdateScript [
   &ParticleState()
   &SkeletalMeshLocation_WithScale(
    Bone / Socket Index:NiagaraInt32=$Particles.SocketIndex
    Bone Sampling Mode:ENiagaraBoneSamplingMode=::NewEnumerator4
    MaskSpawnGroup=false
    Position Sampling:ENiagaraSkelMeshPositionSamplingMode=::NewEnumerator2
    Skeletal Mesh:NiagaraDataInterfaceSkeletalMesh=$User.SkelMesh
   )
   /SetVariables(Particles.Position=&Add_Vector(
     A=&Add_Vector(
      A=&Multiply_VectorByFloat(Vector=$Particles.SkeletalMeshLocation_WithScale.SampledPosition, Float=$System.GlobalGameScale)
      B=&RotateVectorByQuaternion(
       Vector To Rotate=&Multiply_Vector(A=<-0.5 0 0>, B=&VectorFromFloat(Value=$System.GlobalGameScale))
       Quaternion=$Particles.SkeletalMeshLocation_WithScale.SampledRotation
      )
     )
     B=$System.GlobalGameOffset
    )) {isSetParameters=true}
   &ScaleMeshSize(Scale Factor=&Multiply_VectorByFloat(
     Vector=$Output.SkeletalMeshLocation_WithScale.SampledScale
     Float=&MakeCustomFloatFromBool(
      Bool=&Boolean_AndOperation(
       Bool1=&SetBoolByIntComparison(A=$Particles.SocketIndex, B=24, Comparison Type=::NewEnumerator4)
       Bool2=&SetBoolByFloatComparison(
        A=&MakeFloatFromVector(Vector=$Particles.SkeletalMeshLocation_WithScale.SampledPosition, Channel=::NewEnumerator2)
        B=$System.RestingBulletHeight
        Comparison Type=::NewEnumerator6
       )
      )
      True Float=0
      False Float=1
     )
    ))
   &ScaleSpriteSize(
    Scale Sprite Size Mode=::NewEnumerator0
    Uniform Scale Factor=&Multiply_Float(
     A=&MakeCustomFloatFromBool(
      Bool=&SetBoolByFloatComparison(
       A=&MakeFloatFromVector(Vector=$Output.SkeletalMeshLocation_WithScale.SampledScale, Channel=::NewEnumerator0)
       B=0.97
      )
      True Float=1
      False Float=0
     )
     B=&MakeCustomFloatFromBool(
      Bool=&Boolean_AndOperation(
       Bool1=&SetBoolByIntComparison(A=$Particles.SocketIndex, B=24, Comparison Type=::NewEnumerator4)
       Bool2=&SetBoolByFloatComparison(
        A=&MakeFloatFromVector(Vector=$Particles.SkeletalMeshLocation_WithScale.SampledPosition, Channel=::NewEnumerator2)
        B=$System.RestingBulletHeight
        Comparison Type=::NewEnumerator6
       )
      )
      True Float=0
      False Float=1
     )
    )
   )
   /SetVariables(Particles.SubImageIndex=5) {isSetParameters=true}
   /SetVariables(Particles.SpriteAlignment=<0 0.3713 0.9284>, Particles.SpriteFacing=$System.SpriteFacing) {isSetParameters=true}
  ]
  rendererClasses [&NiagaraMeshRendererProperties, &NiagaraSpriteRendererProperties]
  renderers [
   {rendererClass=&NiagaraMeshRendererProperties, rendererIndex=0}
   {rendererClass=&NiagaraSpriteRendererProperties, rendererIndex=1}
  ]
 }
 Explosion {
  emitterUpdateScript [
   &EmitterState()
   &SpawnBurst_Instantaneous(Use Spawn Probability=false, UseLoopCountLimit=false)
  ]
  particleSpawnScript [
   &InitializeParticle(
    Lifetime=5
    Mesh Scale Mode=::NewEnumerator3
    Mesh Uniform Scale=&Multiply_Float(A=1, B=$System.ObjectScale)
    Sprite Size Mode=::NewEnumerator3
    Uniform Sprite Size=&Multiply_Float(A=16, B=$System.ObjectScale)
    UsePositionOffset=false
    Write Material Random=true
   )
   /SetVariables(Particles.SocketIndex=25) {isSetParameters=true, module=SetVariables_#6}
   /SetVariables(Particles.LastExplosionTime=-50) {isSetParameters=true}
  ]
  particleUpdateScript [
   &ParticleState()
   &SkeletalMeshLocation_WithScale(
    Bone / Socket Index:NiagaraInt32=$Particles.SocketIndex
    Bone Sampling Mode:ENiagaraBoneSamplingMode=::NewEnumerator4
    MaskSpawnGroup=false
    Position Sampling:ENiagaraSkelMeshPositionSamplingMode=::NewEnumerator2
    Skeletal Mesh:NiagaraDataInterfaceSkeletalMesh=$User.SkelMesh
   )
   /SetVariables(Particles.Position=&Add_Vector(
     A=&Multiply_VectorByFloat(Vector=$Particles.SkeletalMeshLocation_WithScale.SampledPosition, Float=$System.GlobalGameScale)
     B=$System.GlobalGameOffset
    )) {isSetParameters=true}
   &ScaleMeshSize(Scale Factor=&MakeCustomVectorFromBool(
     Bool=&SetBoolByFloatComparison(
      A=&MakeFloatFromVector(Vector=$Output.SkeletalMeshLocation_WithScale.SampledScale, Channel=::NewEnumerator0)
      B=0.95
     )
     True Vector=<1 1 1>
     False Vector=<0 0 0>
    ))
   &ScaleSpriteSize(
    Scale Sprite Size Mode=::NewEnumerator0
    Uniform Scale Factor=&MakeCustomFloatFromBool(
     Bool=&SetBoolByFloatComparison(
      A=&MakeFloatFromVector(Vector=$Output.SkeletalMeshLocation_WithScale.SampledScale, Channel=::NewEnumerator0)
      B=0.95
     )
     True Float=1
     False Float=0
    )
   )
   &DoOnce(
    Reset:NiagaraBool=&SetBoolByFloatComparison(
     A=&MakeFloatFromVector(Vector=$Output.SkeletalMeshLocation_WithScale.SampledScale, Channel=::NewEnumerator0)
     B=0.95
     Comparison Type=::NewEnumerator6
    )
    Trigger Condition:NiagaraBool=&SetBoolByFloatComparison(
     A=&MakeFloatFromVector(Vector=$Output.SkeletalMeshLocation_WithScale.SampledScale, Channel=::NewEnumerator0)
     B=0.95
    )
   )
   /SetVariables(Particles.LastExplosionTime:NiagaraFloat=&MakeCustomFloatFromBool(Bool=$Particles.DoOnce.Execute, True Float=$Engine.Time, False Float=$Particles.LastExplosionTime)) {isSetParameters=true}
   /SetVariables(Particles.ExplosionFrame:NiagaraInt32=&Add_Integer(
     A=7
     B=&MakeIntFromFloat(Float=&ClampFloat(
       Float=&Multiply_Float(A=&Subtract_Float(A=$Engine.Time, B=$Particles.LastExplosionTime), B=32)
       Max=5
      ))
    )) {isSetParameters=true}
   /SetVariables(Particles.SubImageIndex=&MakeCustomFloatFromBool(
     Bool=&SetBoolByFloatComparison(
      A=&MakeFloatFromVector(Vector=$Output.SkeletalMeshLocation_WithScale.SampledScale, Channel=::NewEnumerator0)
      B=0.1
     )
     True Float=&MakeFloatFromInt(Int=$Particles.ExplosionFrame)
     False Float=15
    )) {isSetParameters=true, module=SetVariables_#7}
   &DynamicMaterialParameters(
    Index 0 Param 1=$Particles.SubImageIndex
    Index 0 Param 4=4
    Module.Param0WriteEnabled=true
    Module.Param1WriteEnabled=true
    Module.Param2WriteEnabled=true
    Module.Param3WriteEnabled=true
   )
   /SetVariables(Particles.SpriteAlignment=<0 0.3713 0.9284>, Particles.SpriteFacing=$System.SpriteFacing) {isSetParameters=true}
  ]
  rendererClasses [&NiagaraMeshRendererProperties, &NiagaraSpriteRendererProperties]
  renderers [
   {rendererClass=&NiagaraMeshRendererProperties, rendererIndex=0}
   {rendererClass=&NiagaraSpriteRendererProperties, rendererIndex=1}
  ]
 }
 Fliers {
  emitterUpdateScript [
   &EmitterState()
   &SpawnBurst_Instantaneous(Spawn Count=4, Use Spawn Probability=false, UseLoopCountLimit=false)
   &SpawnBurst_Instantaneous(Spawn Group=1, Use Spawn Probability=false, UseLoopCountLimit=false) {module=SpawnBurst_Instantaneous001}
   &SpawnBurst_Instantaneous(Spawn Count=2, Spawn Group=2, Use Spawn Probability=false, UseLoopCountLimit=false) {module=SpawnBurst_Instantaneous002}
  ]
  particleSpawnScript [
   &InitializeParticle(
    Lifetime=5
    Mesh Scale Mode=::NewEnumerator3
    Mesh Uniform Scale=&Multiply_Float(A=1, B=$System.ObjectScale)
    Sprite Size Mode=::NewEnumerator3
    Uniform Sprite Size=&Multiply_Float(A=16, B=$System.ObjectScale)
    UsePositionOffset=false
    Write Material Random=true
   )
   /SetVariables(Particles.SocketIndex:NiagaraInt32=&MaskIntBySpawnGroup(
     Masked Int=&Add_Integer(A=&ReturnExecIndex(), B=1)
     Passthrough Int=&MaskIntBySpawnGroup(
      Masked Int=16
      Passthrough Int=&Add_Integer(A=&ReturnExecIndex(), B=17)
      Spawn Group Mask=1
     )
     Spawn Group Mask=0
    )) {isSetParameters=true, module=SetVariables_#6}
   /SetVariables() {isSetParameters=true}
   &SkeletalMeshLocation_WithScale(
    Bone / Socket Index:NiagaraInt32=$Particles.SocketIndex
    Bone Sampling Mode:ENiagaraBoneSamplingMode=::NewEnumerator4
    MaskSpawnGroup=false
    Orientation Sampling:ENiagara_AttributeSamplingApplyOutput=::NewEnumerator1
    Skeletal Mesh:NiagaraDataInterfaceSkeletalMesh=$User.SkelMesh
   ) {module=SkeletalMeshLocation_WithScale001}
   /SetVariables(Particles.Position=&Add_Vector(
     A=&Multiply_VectorByFloat(Vector=$Particles.SkeletalMeshLocation_WithScale001.SampledPosition, Float=$System.GlobalGameScale)
     B=$System.GlobalGameOffset
    )) {isSetParameters=true}
   /SetVariables(Particles.MeshOrientation=&MultiplyQuaternion(
     Quaternion A=$Particles.SkeletalMeshLocation_WithScale001.SampledRotation
     Quaternion B=&MakeQuatFromAxisAngle(Axis=<0 1 0>, AngleInDegrees=90)
    )) {isSetParameters=true}
   /SetVariables(
    Particles.SpriteAlignment=&RotateVectorByQuaternion(Vector To Rotate=<1 0 0>, Quaternion=$Particles.SkeletalMeshLocation_WithScale001.SampledRotation)
    Particles.SpriteFacing=$System.SpriteFacing
   ) {isSetParameters=true}
  ]
  particleUpdateScript [
   &ParticleState()
   &SkeletalMeshLocation_WithScale(
    Bone / Socket Index:NiagaraInt32=$Particles.SocketIndex
    Bone Sampling Mode:ENiagaraBoneSamplingMode=::NewEnumerator4
    MaskSpawnGroup=false
    Orientation Sampling:ENiagara_AttributeSamplingApplyOutput=::NewEnumerator1
    Skeletal Mesh:NiagaraDataInterfaceSkeletalMesh=$User.SkelMesh
   )
   &DoOnce(
    Reset:NiagaraBool=&SetBoolByFloatComparison(
     A=&MakeFloatFromVector(Vector=$Output.SkeletalMeshLocation_WithScale.SampledScale, Channel=::NewEnumerator0)
     B=0.7
    )
    Trigger Condition:NiagaraBool=&SetBoolByFloatComparison(
     A=&MakeFloatFromVector(Vector=$Output.SkeletalMeshLocation_WithScale.SampledScale, Channel=::NewEnumerator0)
     B=0.7
     Comparison Type=::NewEnumerator6
    )
   )
   /SetVariables(Particles.FrameOffset:NiagaraFloat=&MakeCustomFloatFromBool(
     Bool=$Particles.DoOnce.Execute
     True Float=&MakeCustomFloatFromBool(
      Bool=&SetBoolByFloatComparison(A=$Particles.FrameOffset, B=0, Comparison Type=::NewEnumerator4)
      True Float=1
      False Float=0
     )
     False Float=$Particles.FrameOffset
    )) {isSetParameters=true}
   /SetVariables(Particles.Position=&Add_Vector(
     A=&Multiply_VectorByFloat(Vector=$Particles.SkeletalMeshLocation_WithScale.SampledPosition, Float=$System.GlobalGameScale)
     B=$System.GlobalGameOffset
    )) {isSetParameters=true}
   /SetVariables(Particles.MeshOrientation=&MultiplyQuaternion(
     Quaternion A=$Particles.SkeletalMeshLocation_WithScale.SampledRotation
     Quaternion B=&MakeQuatFromAxisAngle(Axis=<0 1 0>, AngleInDegrees=90)
    )) {isSetParameters=true}
   &ScaleMeshSize(Scale Factor=&MakeCustomVectorFromBool(
     Bool=&SetBoolByFloatComparison(
      A=&MakeFloatFromVector(Vector=$Output.SkeletalMeshLocation_WithScale.SampledScale, Channel=::NewEnumerator0)
      B=0.25
     )
     True Vector=<1 1 1>
     False Vector=<0 0 0>
    ))
   &ScaleSpriteSize(
    Scale Sprite Size Mode=::NewEnumerator0
    Uniform Scale Factor=&MakeCustomFloatFromBool(
     Bool=&SetBoolByFloatComparison(
      A=&MakeFloatFromVector(Vector=$Output.SkeletalMeshLocation_WithScale.SampledScale, Channel=::NewEnumerator0)
      B=0.25
     )
     True Float=1
     False Float=0
    )
   )
   /SetVariables(Particles.SubImageIndex=&MakeCustomFloatFromBool(
     Bool=&SetBoolByFloatComparison(
      A=&MakeFloatFromVector(Vector=$Output.SkeletalMeshLocation_WithScale.SampledScale, Channel=::NewEnumerator0)
      B=0.75
     )
     True Float=1
     False Float=&Add_Float(A=2, B=$Particles.FrameOffset)
    )) {isSetParameters=true}
   &DynamicMaterialParameters(
    Index 0 Param 1=$Particles.SubImageIndex
    Index 0 Param 2=&Multiply_Float(A=&MakeFloatFromInt(Int=&SpawnGroup()), B=0.5)
    Index 0 Param 4=4
    Module.Param0WriteEnabled=true
    Module.Param1WriteEnabled=true
    Module.Param2WriteEnabled=true
    Module.Param3WriteEnabled=true
   )
   /SetVariables(
    Particles.SpriteAlignment=&RotateVectorByQuaternion(Vector To Rotate=<1 0 0>, Quaternion=$Particles.SkeletalMeshLocation_WithScale.SampledRotation)
    Particles.SpriteFacing=$System.SpriteFacing
   ) {isSetParameters=true}
  ]
  rendererClasses [&NiagaraMeshRendererProperties, &NiagaraSpriteRendererProperties]
  renderers [
   {rendererClass=&NiagaraMeshRendererProperties, rendererIndex=0}
   {rendererClass=&NiagaraSpriteRendererProperties, rendererIndex=1}
  ]
 }
 FormationShips {
  emitterUpdateScript [
   &EmitterState()
   &SpawnBurst_Instantaneous(Spawn Count=8, Use Spawn Probability=false, UseLoopCountLimit=false)
   &SpawnBurst_Instantaneous(Spawn Count=3, Spawn Group=1, Use Spawn Probability=false, UseLoopCountLimit=false) {module=SpawnBurst_Instantaneous001}
  ]
  particleSpawnScript [
   &InitializeParticle(
    Lifetime=5
    Mesh Scale Mode=::NewEnumerator3
    Mesh Uniform Scale=&Multiply_Float(A=1, B=$System.ObjectScale)
    Sprite Size Mode=::NewEnumerator3
    Uniform Sprite Size=&Multiply_Float(A=16, B=$System.ObjectScale)
    UsePositionOffset=false
    Write Material Random=true
   )
   /SetVariables(Particles.SocketIndex:NiagaraInt32=&MaskIntBySpawnGroup(
     Masked Int=&Add_Integer(A=&ReturnExecIndex(), B=5)
     Passthrough Int=&Add_Integer(A=&ReturnExecIndex(), B=13)
     Spawn Group Mask=0
    )) {isSetParameters=true, module=SetVariables_#6}
   /SetVariables() {isSetParameters=true}
   &SkeletalMeshLocation_WithScale(
    Bone / Socket Index:NiagaraInt32=$Particles.SocketIndex
    Bone Sampling Mode:ENiagaraBoneSamplingMode=::NewEnumerator4
    MaskSpawnGroup=false
    Orientation Sampling:ENiagara_AttributeSamplingApplyOutput=::NewEnumerator1
    Skeletal Mesh:NiagaraDataInterfaceSkeletalMesh=$User.SkelMesh
   ) {module=SkeletalMeshLocation_WithScale001}
   /SetVariables(Particles.Position=&Add_Vector(
     A=&Multiply_VectorByFloat(Vector=$Particles.SkeletalMeshLocation_WithScale001.SampledPosition, Float=$System.GlobalGameScale)
     B=$System.GlobalGameOffset
    )) {isSetParameters=true}
   /SetVariables(Particles.MeshOrientation=&MultiplyQuaternion(
     Quaternion A=$Particles.SkeletalMeshLocation_WithScale001.SampledRotation
     Quaternion B=&MakeQuatFromAxisAngle(Axis=<0 1 0>, AngleInDegrees=90)
    )) {isSetParameters=true}
   /SetVariables(Particles.SpriteAlignment=<0 0.3713 0.9284>, Particles.SpriteFacing=$System.SpriteFacing) {isSetParameters=true}
  ]
  particleUpdateScript [
   &ParticleState()
   &SkeletalMeshLocation_WithScale(
    Bone / Socket Index:NiagaraInt32=$Particles.SocketIndex
    Bone Sampling Mode:ENiagaraBoneSamplingMode=::NewEnumerator4
    MaskSpawnGroup=false
    Orientation Sampling:ENiagara_AttributeSamplingApplyOutput=::NewEnumerator1
    Skeletal Mesh:NiagaraDataInterfaceSkeletalMesh=$User.SkelMesh
   )
   &DoOnce(
    Reset:NiagaraBool=&SetBoolByFloatComparison(
     A=&MakeFloatFromVector(Vector=$Output.SkeletalMeshLocation_WithScale.SampledScale, Channel=::NewEnumerator0)
     B=0.7
    )
    Trigger Condition:NiagaraBool=&SetBoolByFloatComparison(
     A=&MakeFloatFromVector(Vector=$Output.SkeletalMeshLocation_WithScale.SampledScale, Channel=::NewEnumerator0)
     B=0.7
     Comparison Type=::NewEnumerator6
    )
   )
   /SetVariables(Particles.FrameOffset:NiagaraFloat=&MakeCustomFloatFromBool(
     Bool=$Particles.DoOnce.Execute
     True Float=&MakeCustomFloatFromBool(
      Bool=&SetBoolByFloatComparison(A=$Particles.FrameOffset, B=0, Comparison Type=::NewEnumerator4)
      True Float=1
      False Float=0
     )
     False Float=$Particles.FrameOffset
    )) {isSetParameters=true}
   /SetVariables(Particles.Position=&Add_Vector(
     A=&Multiply_VectorByFloat(Vector=$Particles.SkeletalMeshLocation_WithScale.SampledPosition, Float=$System.GlobalGameScale)
     B=$System.GlobalGameOffset
    )) {isSetParameters=true}
   /SetVariables(Particles.MeshOrientation=&MultiplyQuaternion(
     Quaternion A=$Particles.SkeletalMeshLocation_WithScale.SampledRotation
     Quaternion B=&MakeQuatFromAxisAngle(Axis=<0 1 0>, AngleInDegrees=90)
    )) {isSetParameters=true}
   /SetVariables(Particles.SubImageIndex=&MakeCustomFloatFromBool(
     Bool=&SetBoolByFloatComparison(
      A=&MakeFloatFromVector(Vector=$Output.SkeletalMeshLocation_WithScale.SampledScale, Channel=::NewEnumerator0)
      B=0.75
     )
     True Float=1
     False Float=&Add_Float(A=2, B=$Particles.FrameOffset)
    )) {isSetParameters=true}
   &DynamicMaterialParameters(
    Index 0 Param 1=$Particles.SubImageIndex
    Index 0 Param 2=&Multiply_Float(A=&MakeFloatFromInt(Int=&SpawnGroup()), B=0.5)
    Index 0 Param 4=4
    Module.Param0WriteEnabled=true
    Module.Param1WriteEnabled=true
    Module.Param2WriteEnabled=true
    Module.Param3WriteEnabled=true
   )
   /SetVariables(Particles.SpriteAlignment=<0 0.3713 0.9284>, Particles.SpriteFacing=$System.SpriteFacing) {isSetParameters=true}
  ]
  rendererClasses [&NiagaraMeshRendererProperties, &NiagaraSpriteRendererProperties]
  renderers [
   {rendererClass=&NiagaraMeshRendererProperties, rendererIndex=0}
   {rendererClass=&NiagaraSpriteRendererProperties, rendererIndex=1}
  ]
 }
 HighScoreText {
  emitterUpdateScript [
   &EmitterState()
   &SpawnBurst_Instantaneous(Use Spawn Probability=false, UseLoopCountLimit=false)
  ]
  particleSpawnScript [
   &InitializeParticle(
    Lifetime=5
    Mesh Scale Mode=::NewEnumerator3
    Mesh Uniform Scale=&Multiply_Float(A=1, B=$System.ObjectScale)
    UsePositionOffset=false
    Write Material Random=true
   )
   &SkeletalMeshLocation(
    Bone / Socket Index=27
    Bone Sampling Mode=::NewEnumerator4
    MaskSpawnGroup=false
    Position Sampling=::NewEnumerator2
    Sampled Position Offset=<1.75 0 0>
    Skeletal Mesh=$User.SkelMesh
   ) {module=SkeletalMeshLocation001}
   /SetVariables(Particles.Position=&Add_Vector(
     A=&Add_Vector(
      A=&Multiply_VectorByFloat(
       Vector=Linked() {linkedVariable {name=Particles.SkeletalMeshLocation001.SampledPosition, type=NiagaraPosition}}
       Float=$System.GlobalGameScale
      )
      B=&RotateVectorByQuaternion(
       Vector To Rotate=&Multiply_Vector(A=<1 0 16>, B=&VectorFromFloat(Value=$System.GlobalGameScale))
       Quaternion=$Particles.SkeletalMeshLocation001.SampledRotation
      )
     )
     B=$System.GlobalGameOffset
    )) {isSetParameters=true, module=SetVariables_#4}
   /SetVariables(Particles.MeshOrientation=&MultiplyQuaternion(
     Quaternion A=$Particles.SkeletalMeshLocation001.SampledRotation
     Quaternion B=&MultiplyQuaternion(
      Quaternion A=&MakeQuatFromAxisAngle(Axis=<1 0 0>, AngleInDegrees=-90)
      Quaternion B=&MakeQuatFromAxisAngle(Axis=<0 1 0>, AngleInDegrees=90)
     )
    )) {isSetParameters=true, module=SetVariables_#5}
  ]
  particleUpdateScript [&ParticleState()]
  rendererClasses [&NiagaraMeshRendererProperties]
  renderers [{rendererClass=&NiagaraMeshRendererProperties, rendererIndex=0}]
 }
 PlayerShip {
  emitterUpdateScript [
   &EmitterState(
    Inactive Response=::NewEnumerator1
    Life Cycle Mode=::NewEnumerator1
    Loop Behavior=::NewEnumerator1
    Loop Duration Mode=::NewEnumerator1
    UseLoopDelay=false
   )
   &SpawnBurst_Instantaneous(Use Spawn Probability=false, UseLoopCountLimit=false)
  ]
  particleSpawnScript [
   &InitializeParticle(
    Lifetime=5
    Mesh Scale Mode=::NewEnumerator3
    Mesh Uniform Scale=&Multiply_Float(A=1, B=$System.ObjectScale)
    Sprite Size Mode=::NewEnumerator3
    Uniform Sprite Size=&Multiply_Float(A=16, B=$System.ObjectScale)
    UsePositionOffset=true
    Write Material Random=true
   )
   /SetVariables() {isSetParameters=true, module=SetVariables_#7}
   /SetVariables(
    Particles.BulletColor:LinearColor=<rgba 0.8714 0.8308 0.0595 1>
    Particles.BulletScale:Vector3f=<0.5 0.5 1.75>
    Particles.BulletSize:Vector2f=<8 8>
    Particles.BulletSubUV=5
   ) {isSetParameters=true}
   &SkeletalMeshLocation(
    Bone Sampling Mode=::NewEnumerator4
    MaskSpawnGroup=false
    Orientation Sampling=::NewEnumerator1
    Position Sampling=::NewEnumerator2
    Skeletal Mesh=$User.SkelMesh
   ) {module=SkeletalMeshLocation001}
   &SkeletalMeshLocation_WithScale(
    Bone / Socket Index=24
    Bone Sampling Mode:ENiagaraBoneSamplingMode=::NewEnumerator4
    MaskSpawnGroup=false
    Orientation Sampling:ENiagara_AttributeSamplingApplyOutput=::NewEnumerator1
    Position Sampling:ENiagaraSkelMeshPositionSamplingMode=::NewEnumerator2
    Skeletal Mesh:NiagaraDataInterfaceSkeletalMesh=$User.SkelMesh
   ) {module=SkeletalMeshLocation_WithScale001}
   /SetVariables(Particles.Position=&Add_Vector(
     A=&Add_Vector(
      A=&Multiply_VectorByFloat(
       Vector=Linked() {linkedVariable {name=Particles.SkeletalMeshLocation001.SampledPosition, type=NiagaraPosition}}
       Float=$System.GlobalGameScale
      )
      B=&RotateVectorByQuaternion(
       Vector To Rotate=&Multiply_Vector(A=<0 0 0>, B=&VectorFromFloat(Value=$System.GlobalGameScale))
       Quaternion=$Particles.SkeletalMeshLocation001.SampledRotation
      )
     )
     B=$System.GlobalGameOffset
    )) {isSetParameters=true}
  ]
  particleUpdateScript [
   &ParticleState(Kill Particles When Lifetime Has Elapsed=false, Let Infinitely Lived Particles Die When Emitter Deactivates=true)
   &SkeletalMeshLocation(
    Bone Sampling Mode=::NewEnumerator4
    MaskSpawnGroup=false
    Orientation Sampling=::NewEnumerator1
    Position Sampling=::NewEnumerator2
    Skeletal Mesh=$User.SkelMesh
   )
   &SkeletalMeshLocation_WithScale(
    Bone / Socket Index=24
    Bone Sampling Mode:ENiagaraBoneSamplingMode=::NewEnumerator4
    MaskSpawnGroup=false
    Orientation Sampling:ENiagara_AttributeSamplingApplyOutput=::NewEnumerator1
    Position Sampling:ENiagaraSkelMeshPositionSamplingMode=::NewEnumerator2
    Skeletal Mesh:NiagaraDataInterfaceSkeletalMesh=$User.SkelMesh
   )
   /SetVariables(Particles.Position=&Add_Vector(
     A=&Add_Vector(
      A=&Multiply_VectorByFloat(Vector=Linked() {linkedVariable {name=Particles.SkeletalMeshLocation.SampledPosition, type=NiagaraPosition}}, Float=$System.GlobalGameScale)
      B=&RotateVectorByQuaternion(
       Vector To Rotate=&Multiply_Vector(A=<0 0 0>, B=&VectorFromFloat(Value=$System.GlobalGameScale))
       Quaternion=$Particles.SkeletalMeshLocation.SampledRotation
      )
     )
     B=$System.GlobalGameOffset
    )) {isSetParameters=true}
   /SetVariables(Particles.MeshOrientation=&MultiplyQuaternion(
     Quaternion A=$Particles.SkeletalMeshLocation.SampledRotation
     Quaternion B=&MakeQuatFromAxisAngle(Axis=<0 1 0>, AngleInDegrees=90)
    )) {isSetParameters=true}
   /SetVariables(Particles.SpriteAlignment=<0 0.3713 0.9284>, Particles.SpriteFacing=$System.SpriteFacing) {isSetParameters=true}
   /SetVariables(
    Particles.BulletPosition:NiagaraPosition=&Add_Vector(
     A=&Add_Vector(
      A=&Multiply_VectorByFloat(Vector=Linked() {linkedVariable {name=Particles.Position, type=NiagaraPosition}}, Float=$System.GlobalGameScale)
      B=&RotateVectorByQuaternion(
       Vector To Rotate=&Multiply_Vector(A=<4.5 0 0>, B=&VectorFromFloat(Value=$System.GlobalGameScale))
       Quaternion=$Particles.SkeletalMeshLocation.SampledRotation
      )
     )
     B=$System.GlobalGameOffset
    )
    Particles.BulletScale:Vector3f=&Multiply_VectorByFloat(
     Vector=$Particles.Initial.BulletScale
     Float=&MakeCustomFloatFromBool(
      Bool=&SetBoolByFloatComparison(
       A=&MakeFloatFromVector(Vector=$Particles.SkeletalMeshLocation_WithScale.SampledPosition, Channel=::NewEnumerator2)
       B=$System.RestingBulletHeight
      )
      True Float=0
      False Float=1
     )
    )
    Particles.BulletSize:Vector2f=&Multiply_Vector2D_ByFloat(
     Float=&MakeCustomFloatFromBool(
      Bool=&SetBoolByFloatComparison(
       A=&MakeFloatFromVector(Vector=$Particles.SkeletalMeshLocation_WithScale.SampledPosition, Channel=::NewEnumerator2)
       B=$System.RestingBulletHeight
      )
      True Float=0
      False Float=1
     )
     Vector2D=$Particles.Initial.BulletSize
    )
   ) {isSetParameters=true}
   &DynamicMaterialParameters(
    Index 0 Param 1=$Particles.SubImageIndex
    Index 0 Param 4=4
    Index 3 Param 1=&MakeFloatFromVector(Vector=$Particles.SkeletalMeshLocation_WithScale.SampledPosition, Channel=::NewEnumerator2)
    Module.Param0WriteEnabled=true
    Module.Param0WriteEnabled3=true
    Module.Param1WriteEnabled=true
    Module.Param1WriteEnabled3=true
    Module.Param2WriteEnabled=true
    Module.Param2WriteEnabled3=true
    Module.Param3WriteEnabled=true
    Module.Param3WriteEnabled3=true
    Write Parameter Index 3=true
   )
  ]
  rendererClasses [&NiagaraMeshRendererProperties, &NiagaraSpriteRendererProperties]
  renderers [
   {rendererClass=&NiagaraMeshRendererProperties, rendererIndex=0}
   {rendererClass=&NiagaraSpriteRendererProperties, rendererIndex=1}
   {rendererClass=&NiagaraMeshRendererProperties, rendererIndex=2}
   {rendererClass=&NiagaraSpriteRendererProperties, rendererIndex=3}
  ]
 }
 Score {
  emitterSpawnScript [/SetVariables() {isSetParameters=true}]
  emitterUpdateScript [
   &EmitterState()
   &SpawnBurst_Instantaneous(Spawn Count=4, Use Spawn Probability=false, UseLoopCountLimit=false)
   &DoOnce(
    Reset:NiagaraBool=&SetBoolByFloatComparison(A=$User.PointsCurve, B=0.1, Comparison Type=::NewEnumerator6)
    Trigger Condition:NiagaraBool=&SetBoolByFloatComparison(A=$User.PointsCurve, B=0.1)
   )
   /SetVariables(Emitter.Score:NiagaraInt32=&self:ClampInt(
     IntToClamp:NiagaraInt32=&Add_Integer(
      A=$Emitter.Score
      B=&MakeCustomIntFromBool(
       Bool=$Emitter.DoOnce.Execute
       True INT=&SelectIntFromArray(
        Direct Array Index=&MakeIntFromFloat(Float=$User.PointsCurve)
        Int Selection Array=DataInterface() {propertyValues {IntData [0,60,80,100]}}
       )
       False INT=0
      )
     )
     Min:NiagaraInt32=0
     Max:NiagaraInt32=9940
    )) {isSetParameters=true}
  ]
  particleSpawnScript [
   &InitializeParticle(
    Color=<rgba 0.7991 0.0176 0.0091 1>
    Lifetime=5
    Mesh Scale Mode=::NewEnumerator3
    Mesh Uniform Scale=&Multiply_Float(A=1, B=$System.ObjectScale)
    Sprite Size Mode=::NewEnumerator3
    Uniform Sprite Size=&Multiply_Float(A=10, B=$System.ObjectScale)
    UsePositionOffset=false
    Write Material Random=true
   )
   &SkeletalMeshLocation(
    Bone / Socket Index=26
    Bone Sampling Mode=::NewEnumerator4
    MaskSpawnGroup=false
    Orientation Sampling=::NewEnumerator1
    Position Sampling=::NewEnumerator2
    Skeletal Mesh=$User.SkelMesh
    Use Sampled Mesh Attribute as Ribbon ID=false
   )
   /SetVariables(Particles.Position=&Add_Vector(
     A=&Add_Vector(
      A=&Multiply_VectorByFloat(Vector=Linked() {linkedVariable {name=Particles.SkeletalMeshLocation.SampledPosition, type=NiagaraPosition}}, Float=$System.GlobalGameScale)
      B=&RotateVectorByQuaternion(
       Vector To Rotate=&Multiply_VectorByFloat(
        Vector=&Add_Vector(
         A=&Lerp_Vector(A=<1 0 2>, B=<1 0 -12>, Alpha=&ReturnNormalizedExecIndex())
         B=<0 0 -18>
        )
        Float=$System.GlobalGameScale
       )
       Quaternion=$Particles.SkeletalMeshLocation.SampledRotation
      )
     )
     B=$System.GlobalGameOffset
    )) {isSetParameters=true}
   /SetVariables(Particles.MeshOrientation=&MultiplyQuaternion(
     Quaternion A=$Particles.SkeletalMeshLocation.SampledRotation
     Quaternion B=&MultiplyQuaternion(
      Quaternion A=&MakeQuatFromAxisAngle(Axis=<0 1 0>, AngleInDegrees=90)
      Quaternion B=&MakeQuatFromAxisAngle(Axis=<0 0 1>, AngleInDegrees=180)
     )
    )) {isSetParameters=true}
   /SetVariables(Particles.Slot:NiagaraInt32=&ReturnExecIndex()) {isSetParameters=true}
   /SetVariables(Particles.SubImageIndex=10) {isSetParameters=true}
   /SetVariables(Particles.SecondaryPosition:NiagaraPosition=&Add_Vector(
     A=Linked() {linkedVariable {name=Particles.Position, type=NiagaraPosition}}
     B=&RotateVectorByQuaternion(
      Vector To Rotate=&Multiply_VectorByFloat(Vector=<0 0 37>, Float=$System.GlobalGameScale)
      Quaternion=$Particles.SkeletalMeshLocation.SampledRotation
     )
    )) {isSetParameters=true}
  ]
  particleUpdateScript [
   &ParticleState()
   /SetVariables(Particles.SubImageIndex=&MakeFloatFromInt(Int=&ModuloInt(
      A=/Niagara/DynamicInputs/Divide/Divide_Int(
       A:NiagaraInt32=$Emitter.Score
       B:NiagaraInt32=&self:PowerInt(A:NiagaraInt32=10, B:NiagaraInt32=$Particles.Slot)
      )
      B=10
     ))) {isSetParameters=true}
   /SetVariables(Particles.Scale=&MakeCustomVectorFromBool(
     Bool=&Boolean_OrOperation(
      Bool1=&SetBoolByFloatComparison(
       A=&MakeFloatFromInt(Int=$Emitter.Score)
       B=&Power(a=10, b=&MakeFloatFromInt(Int=$Particles.Slot))
       Comparison Type=::NewEnumerator2
      )
      Bool2=&SetBoolByIntComparison(A=$Particles.Slot, B=1, Comparison Type=::NewEnumerator7)
     )
     True Vector=$Particles.Initial.Scale
     False Vector=<0 0 0>
    )) {isSetParameters=true}
   /SetVariables(Particles.SpriteSize=&MakeCustomVector2DFromBool(
     Bool=&Boolean_OrOperation(
      Bool1=&SetBoolByFloatComparison(
       A=&MakeFloatFromInt(Int=$Emitter.Score)
       B=&Power(a=10, b=&MakeFloatFromInt(Int=$Particles.Slot))
       Comparison Type=::NewEnumerator2
      )
      Bool2=&SetBoolByIntComparison(A=$Particles.Slot, B=1, Comparison Type=::NewEnumerator7)
     )
     True Vector2D=$Particles.Initial.SpriteSize
     False Vector2D=<0 0>
    )) {isSetParameters=true}
   &DynamicMaterialParameters(
    Index 0 Param 1=$Particles.SubImageIndex
    Index 0 Param 4=4
    Module.Param0WriteEnabled=true
    Module.Param1WriteEnabled=true
    Module.Param2WriteEnabled=true
    Module.Param3WriteEnabled=true
   )
   /SetVariables(Particles.MeshIndex=&MakeIntFromFloat(Float=$Particles.SubImageIndex)) {isSetParameters=true}
   /SetVariables(Particles.SpriteAlignment=<0 0.3713 0.9284>, Particles.SpriteFacing=$System.SpriteFacing) {isSetParameters=true}
  ]
  rendererClasses [&NiagaraMeshRendererProperties, &NiagaraSpriteRendererProperties]
  renderers [
   {rendererClass=&NiagaraMeshRendererProperties, rendererIndex=0}
   {rendererClass=&NiagaraSpriteRendererProperties, rendererIndex=1}
   {rendererClass=&NiagaraMeshRendererProperties, rendererIndex=2}
   {rendererClass=&NiagaraSpriteRendererProperties, rendererIndex=3}
  ]
 }
 Yellow {
  emitterUpdateScript [
   &EmitterState()
   &SpawnBurst_Instantaneous(Use Spawn Probability=false, UseLoopCountLimit=false)
  ]
  particleSpawnScript [
   &InitializeParticle(
    Lifetime=5
    Mesh Scale Mode=::NewEnumerator3
    Mesh Uniform Scale=&Multiply_Float(A=1, B=$System.ObjectScale)
    Sprite Size Mode=::NewEnumerator3
    Uniform Sprite Size=&Multiply_Float(A=16, B=$System.ObjectScale)
    UsePositionOffset=false
    Write Material Random=true
   )
   /SetVariables(Particles.SocketIndex=19) {isSetParameters=true, module=SetVariables_#6}
   &SkeletalMeshLocation_WithScale(
    Bone / Socket Index:NiagaraInt32=$Particles.SocketIndex
    Bone Sampling Mode:ENiagaraBoneSamplingMode=::NewEnumerator4
    MaskSpawnGroup=false
    Position Sampling:ENiagaraSkelMeshPositionSamplingMode=::NewEnumerator2
    Skeletal Mesh:NiagaraDataInterfaceSkeletalMesh=$User.SkelMesh
   ) {module=SkeletalMeshLocation_WithScale001}
   /SetVariables(Particles.Position=&Add_Vector(
     A=&Multiply_VectorByFloat(Vector=$Particles.SkeletalMeshLocation_WithScale001.SampledPosition, Float=$System.GlobalGameScale)
     B=$System.GlobalGameOffset
    )) {isSetParameters=true}
   /SetVariables(Particles.MeshOrientation=&MultiplyQuaternion(
     Quaternion A=$Particles.SkeletalMeshLocation_WithScale001.SampledRotation
     Quaternion B=&MakeQuatFromAxisAngle(Axis=<0 1 0>, AngleInDegrees=90)
    )) {isSetParameters=true}
   /SetVariables(Particles.SubImageIndex=4) {isSetParameters=true}
   &DynamicMaterialParameters(
    Index 0 Param 1=$Particles.SubImageIndex
    Index 0 Param 4=4
    Module.Param0WriteEnabled=true
    Module.Param1WriteEnabled=true
    Module.Param2WriteEnabled=true
    Module.Param3WriteEnabled=true
   ) {module=DynamicMaterialParameters001}
   /SetVariables(
    Particles.SpriteAlignment=&RotateVectorByQuaternion(Vector To Rotate=<1 0 0>, Quaternion=$Particles.SkeletalMeshLocation_WithScale001.SampledRotation)
    Particles.SpriteFacing=$System.SpriteFacing
   ) {isSetParameters=true}
  ]
  particleUpdateScript [
   &ParticleState()
   &SkeletalMeshLocation_WithScale(
    Bone / Socket Index:NiagaraInt32=$Particles.SocketIndex
    Bone Sampling Mode:ENiagaraBoneSamplingMode=::NewEnumerator4
    MaskSpawnGroup=false
    Position Sampling:ENiagaraSkelMeshPositionSamplingMode=::NewEnumerator2
    Skeletal Mesh:NiagaraDataInterfaceSkeletalMesh=$User.SkelMesh
   )
   /SetVariables(Particles.Position=&Add_Vector(
     A=&Multiply_VectorByFloat(Vector=$Particles.SkeletalMeshLocation_WithScale.SampledPosition, Float=$System.GlobalGameScale)
     B=$System.GlobalGameOffset
    )) {isSetParameters=true}
   /SetVariables(Particles.MeshOrientation=&MultiplyQuaternion(
     Quaternion A=$Particles.SkeletalMeshLocation_WithScale.SampledRotation
     Quaternion B=&MakeQuatFromAxisAngle(Axis=<0 1 0>, AngleInDegrees=90)
    )) {isSetParameters=true}
   /SetVariables(Particles.SubImageIndex=4) {isSetParameters=true, module=SetVariables_#7}
   &ScaleSpriteSize(
    Scale Sprite Size Mode=::NewEnumerator0
    Uniform Scale Factor=&MakeCustomFloatFromBool(
     Bool=&SetBoolByFloatComparison(
      A=&MakeFloatFromVector(Vector=$Output.SkeletalMeshLocation_WithScale.SampledScale, Channel=::NewEnumerator0)
      B=0.97
     )
     True Float=1
     False Float=0
    )
   )
   &ScaleMeshSize(Scale Factor=&MakeCustomVectorFromBool(
     Bool=&SetBoolByFloatComparison(
      A=&MakeFloatFromVector(Vector=$Output.SkeletalMeshLocation_WithScale.SampledScale, Channel=::NewEnumerator0)
      B=0.97
     )
     True Vector=<1 1 1>
     False Vector=<0 0 0>
    ))
   &DynamicMaterialParameters(
    Index 0 Param 1=$Particles.SubImageIndex
    Index 0 Param 4=4
    Module.Param0WriteEnabled=true
    Module.Param1WriteEnabled=true
    Module.Param2WriteEnabled=true
    Module.Param3WriteEnabled=true
   )
   /SetVariables(
    Particles.SpriteAlignment=&RotateVectorByQuaternion(Vector To Rotate=<1 0 0>, Quaternion=$Particles.SkeletalMeshLocation_WithScale.SampledRotation)
    Particles.SpriteFacing=$System.SpriteFacing
   ) {isSetParameters=true}
  ]
  rendererClasses [&NiagaraMeshRendererProperties, &NiagaraSpriteRendererProperties]
  renderers [
   {rendererClass=&NiagaraMeshRendererProperties, rendererIndex=0}
   {rendererClass=&NiagaraSpriteRendererProperties, rendererIndex=1}
  ]
 }
}
summary {
 systemName=NS_InterstellarSprites
 emitters [
  {emitterName=PlayerShip, rendererClasses [&NiagaraMeshRendererProperties, &NiagaraSpriteRendererProperties]}
  {emitterName=FormationShips, rendererClasses [&NiagaraMeshRendererProperties, &NiagaraSpriteRendererProperties]}
  {emitterName=Yellow, rendererClasses [&NiagaraMeshRendererProperties, &NiagaraSpriteRendererProperties]}
  {emitterName=Bullets, rendererClasses [&NiagaraMeshRendererProperties, &NiagaraSpriteRendererProperties]}
  {emitterName=Explosion, rendererClasses [&NiagaraMeshRendererProperties, &NiagaraSpriteRendererProperties]}
  {emitterName=HighScoreText, rendererClasses [&NiagaraMeshRendererProperties]}
  {emitterName=Score, rendererClasses [&NiagaraMeshRendererProperties, &NiagaraSpriteRendererProperties]}
  {emitterName=Fliers, rendererClasses [&NiagaraMeshRendererProperties, &NiagaraSpriteRendererProperties]}
  {emitterName=1UPText, rendererClasses [&NiagaraMeshRendererProperties]}
 ]
 userVariables [
  User.NormalizedNotifyProgress {type=NiagaraFloat, defaultValue=0, description=""}
  User.PointsCurve {type=NiagaraFloat, defaultValue=0, description=""}
  User.SkelMesh {
   type=NiagaraDataInterfaceSkeletalMesh
   defaultValue {
    struct=NiagaraExt_VariableValue_DataInterface
    value {dataInterface=&self:NiagaraDataInterfaceSkeletalMesh_0}
   }
  }
 ]
}
